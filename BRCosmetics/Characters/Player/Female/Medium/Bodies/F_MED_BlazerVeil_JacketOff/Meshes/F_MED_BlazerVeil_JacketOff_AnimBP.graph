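# Animation Blueprint for the female BlazerVeil JacketOff cosmetic body, adding secondary motion to the pose copied from the parent mesh.
# Chained AnimDynamics and bone constraints drive ankle, tie, collar and back skirt bones, then a skydive-aware Control Rig and additive tie rotations adjust the result.
# Update logic sets skirt gravity via a deprecated helper, the skydive value, and a front-end idle curve weighting a DefaultSlot blend over collar bones.
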AnimBlueprint /BRCosmetics/Characters/Player/Female/Medium/Bodies/F_MED_BlazerVeil_JacketOff/Meshes/F_MED_BlazerVeil_JacketOff_AnimBP
target_skeleton=&/F_MED_BlazerVeil_JacketOff_Skeleton
graphs [
 AnimGraph {nodes [
   n1 = Anim Dynamics - Bone: dyn_ankle_top!(
    ComponentPose=Local To Component(LocalPose=Copy Pose From Mesh() {bUseAttachedParent=true})
    Alpha=1
    GravityOverride:Vector="0,0,-980"
    ExternalForce="0,0,0"
   ) -> (Pose) {
    angularDampingOverride=0.99
    angularSpringConstant=100
    bAngularSpring=true
    bOverrideAngularDamping=true
    bUseGravityOverride=true
    bUseSphericalLimits=true
    simulationSpace=RootRelative
    alphaBoolBlend {blendOption=Linear}
    boundBone {boneName=dyn_ankle_top}
    componentAppliedLinearAccClamp=<2000 2000 2000>
    componentLinearAccScale=<0.8 0.8 0.8>
    componentLinearVelScale=<0.8 0.8 0.8>
    gravityOverride {z=-980}
    physicsBodyDefinitions [{
      collisionType=CustomSphere
      sphereCollisionRadius=4.5
      boundBone {boneName=dyn_ankle_top}
      boxExtents=<5 25 25>
      constraintSetup {angularLimitsMax=<1 20 20>, angularLimitsMin=<-1 -20 -20>, angularTarget=<1 0 0>}
      localJointOffset=<3 0 0>
     }]
    relativeSpaceBone {boneName=pelvis}
   }
   n2 = Anim Dynamics - Bone: dyn_ankle_bot!(ComponentPose=@n1, Alpha=1, GravityOverride:Vector="0,0,-980", ExternalForce="0,0,0") -> (Pose) {
    angularDampingOverride=0.99
    angularSpringConstant=100
    bAngularSpring=true
    bOverrideAngularDamping=true
    bUseGravityOverride=true
    bUseSphericalLimits=true
    simulationSpace=RootRelative
    alphaBoolBlend {blendOption=Linear}
    boundBone {boneName=dyn_ankle_bot}
    componentAppliedLinearAccClamp=<2000 2000 2000>
    componentLinearAccScale=<0.8 0.8 0.8>
    componentLinearVelScale=<0.8 0.8 0.8>
    gravityOverride {z=-980}
    physicsBodyDefinitions [{
      collisionType=CustomSphere
      sphereCollisionRadius=4.5
      boundBone {boneName=dyn_ankle_bot}
      boxExtents=<5 25 25>
      constraintSetup {angularLimitsMax=<1 20 20>, angularLimitsMin=<-1 -20 -20>, angularTarget=<1 0 0>}
      localJointOffset=<3 0 0>
     }]
    relativeSpaceBone {boneName=pelvis}
   }
   n3 = "Constraint - dyn_tie_01 (PPPPPP)"!(
    ComponentPose=@n2
    Alpha=0.8
    ConstraintWeights_0=0.128
    ConstraintWeights_1:float=0.8503
    ConstraintWeights_2:float=0.006
    ConstraintWeights_3:float=0.0053
    ConstraintWeights_4:float=0.0104
    ConstraintWeights_5:float=0
   ) -> (Pose) {
    alphaBoolBlend {blendOption=Linear}
    boneToModify {boneName=dyn_tie_01}
    constraintSetup [
     {targetBone {boneName=spine_05}}
     {targetBone {boneName=spine_04}}
     {targetBone {boneName=spine_03}}
     {targetBone {boneName=spine_02}}
     {targetBone {boneName=deform_pec_l}}
     {targetBone {boneName=deform_pec_r}}
    ]
    constraintWeights [1,1,1,1,1,1]
   }
   n4 = Anim Dynamics - Bone: dyn_tie_01!(ComponentPose=@n3, Alpha=1, GravityOverride:Vector="0,0,-980", ExternalForce="0,0,0") -> (Pose) {
    angularDampingOverride=0.99
    angularSpringConstant=100
    bAngularSpring=true
    bOverrideAngularDamping=true
    bUseGravityOverride=true
    bUseSphericalLimits=true
    simulationSpace=RootRelative
    alphaBoolBlend {blendOption=Linear}
    boundBone {boneName=dyn_tie_01}
    componentAppliedLinearAccClamp=<3000 3000 3000>
    componentLinearAccScale=<0.6 0.6 0.6>
    componentLinearVelScale=<0.6 0.6 0.6>
    gravityOverride {z=-980}
    physicsBodyDefinitions [{
      collisionType=CustomSphere
      sphereCollisionRadius=4.5
      boundBone {boneName=dyn_tie_01}
      boxExtents=<25 15 15>
      constraintSetup {angularLimitsMax=<1 20 20>, angularLimitsMin=<-1 -20 -20>, angularTarget=<1 0 0>}
      localJointOffset=<2 0 0>
     }]
    relativeSpaceBone {boneName=pelvis}
    sphericalLimits [
     {limitRadius=8, drivingBone {boneName=spine_04}, sphereLocalOffset=<0 0 0>}
     {limitRadius=8, drivingBone {boneName=spine_05}, sphereLocalOffset=<0 0 0>}
    ]
   }
   n5 = Anim Dynamics - Bone: dyn_tie_02!(ComponentPose=@n4, Alpha=1, GravityOverride:Vector="0,0,-980", ExternalForce="0,0,0") -> (Pose) {
    angularDampingOverride=0.99
    angularSpringConstant=100
    bAngularSpring=true
    bOverrideAngularDamping=true
    bUseGravityOverride=true
    bUseSphericalLimits=true
    simulationSpace=RootRelative
    alphaBoolBlend {blendOption=Linear}
    boundBone {boneName=dyn_tie_02}
    componentAppliedLinearAccClamp=<3000 3000 3000>
    componentLinearAccScale=<0.6 0.6 0.6>
    componentLinearVelScale=<0.6 0.6 0.6>
    gravityOverride {z=-980}
    physicsBodyDefinitions [{
      collisionType=CustomSphere
      sphereCollisionRadius=4.5
      boundBone {boneName=dyn_tie_02}
      boxExtents=<25 15 15>
      constraintSetup {angularLimitsMax=<1 20 20>, angularLimitsMin=<-1 -20 -20>, angularTarget=<1 0 0>}
      localJointOffset=<2 0 0>
     }]
    relativeSpaceBone {boneName=pelvis}
    sphericalLimits [
     {limitRadius=8, drivingBone {boneName=spine_04}, sphereLocalOffset=<0 0 0>}
     {limitRadius=8, drivingBone {boneName=spine_05}, sphereLocalOffset=<0 0 0>}
    ]
   }
   n6 = Anim Dynamics - Bone: dyn_tie_03!(ComponentPose=@n5, Alpha=1, GravityOverride:Vector="0,0,-980", ExternalForce="0,0,0") -> (Pose) {
    angularDampingOverride=0.99
    angularSpringConstant=100
    bAngularSpring=true
    bOverrideAngularDamping=true
    bUseGravityOverride=true
    bUseSphericalLimits=true
    simulationSpace=RootRelative
    alphaBoolBlend {blendOption=Linear}
    boundBone {boneName=dyn_tie_03}
    componentAppliedLinearAccClamp=<3000 3000 3000>
    componentLinearAccScale=<0.6 0.6 0.6>
    componentLinearVelScale=<0.6 0.6 0.6>
    gravityOverride {z=-980}
    physicsBodyDefinitions [{
      collisionType=CustomSphere
      sphereCollisionRadius=3.5
      boundBone {boneName=dyn_tie_03}
      boxExtents=<25 15 15>
      constraintSetup {angularLimitsMax=<1 20 20>, angularLimitsMin=<-1 -20 -20>, angularTarget=<1 0 0>}
      localJointOffset=<2 0 0>
     }]
    relativeSpaceBone {boneName=pelvis}
    sphericalLimits [
     {limitRadius=8, drivingBone {boneName=spine_02}, sphereLocalOffset=<0 0 0>}
     {limitRadius=8, drivingBone {boneName=spine_03}, sphereLocalOffset=<0 0 0>}
    ]
   }
   n7 = Anim Dynamics - Bone: dyn_tie_04!(ComponentPose=@n6, Alpha=1, GravityOverride:Vector="0,0,-980", ExternalForce="0,0,0") -> (Pose) {
    angularDampingOverride=0.99
    angularSpringConstant=100
    bAngularSpring=true
    bOverrideAngularDamping=true
    bUseGravityOverride=true
    bUseSphericalLimits=true
    simulationSpace=RootRelative
    alphaBoolBlend {blendOption=Linear}
    boundBone {boneName=dyn_tie_04}
    componentAppliedLinearAccClamp=<3000 3000 3000>
    componentLinearAccScale=<0.6 0.6 0.6>
    componentLinearVelScale=<0.6 0.6 0.6>
    gravityOverride {z=-980}
    physicsBodyDefinitions [{
      collisionType=CustomSphere
      sphereCollisionRadius=3.5
      boundBone {boneName=dyn_tie_04}
      boxExtents=<25 15 15>
      constraintSetup {angularLimitsMax=<1 20 20>, angularLimitsMin=<-1 -20 -20>, angularTarget=<1 0 0>}
      localJointOffset=<2 0 0>
     }]
    relativeSpaceBone {boneName=pelvis}
    sphericalLimits [
     {limitRadius=8, drivingBone {boneName=spine_02}, sphereLocalOffset=<0 0 0>}
     {limitRadius=8, drivingBone {boneName=spine_03}, sphereLocalOffset=<0 0 0>}
    ]
   }
   n8 = Anim Dynamics - Bone: dyn_collar_r_01(ComponentPose=@n7, Alpha=1, ExternalForce="0,0,0") {
    angularDampingOverride=1
    angularSpringConstant=20000
    bAngularSpring=true
    bOverrideAngularDamping=true
    bUseGravityOverride=true
    gravityScale=0.35
    simulationSpace=RootRelative
    alphaBoolBlend {blendOption=Linear}
    boundBone {boneName=dyn_collar_r_01}
    componentLinearAccScale=<0.7 0.7 0.7>
    componentLinearVelScale=<0.7 0.7 0.7>
    gravityOverride {z=-980}
    physicsBodyDefinitions [{
      sphereCollisionRadius=0
      boundBone {boneName=dyn_collar_r_01}
      boxExtents=<20 30 30>
      constraintSetup {angularLimitsMax=<0 5 10>, angularLimitsMin=<0 -2 -10>, angularTarget=<1 0 0>}
      localJointOffset=<4 0 0>
     }]
   }
   n9 = Anim Dynamics - Bone: dyn_collar_r_02(ComponentPose=@n8, Alpha=1, ExternalForce="0,0,0") {
    angularDampingOverride=1
    angularSpringConstant=20000
    bAngularSpring=true
    bOverrideAngularDamping=true
    bUseGravityOverride=true
    gravityScale=0.35
    simulationSpace=RootRelative
    alphaBoolBlend {blendOption=Linear}
    boundBone {boneName=dyn_collar_r_02}
    componentLinearAccScale=<0.7 0.7 0.7>
    componentLinearVelScale=<0.7 0.7 0.7>
    gravityOverride {z=-980}
    physicsBodyDefinitions [{
      sphereCollisionRadius=0
      boundBone {boneName=dyn_collar_r_02}
      boxExtents=<20 30 30>
      constraintSetup {angularLimitsMax=<0 5 10>, angularLimitsMin=<0 -2 -10>, angularTarget=<1 0 0>}
      localJointOffset=<4 0 0>
     }]
   }
   n10 = Anim Dynamics - Bone: dyn_collar_l_01(ComponentPose=@n9, Alpha=1, ExternalForce="0,0,0") {
    angularDampingOverride=1
    angularSpringConstant=20000
    bAngularSpring=true
    bOverrideAngularDamping=true
    bUseGravityOverride=true
    gravityScale=0.35
    simulationSpace=RootRelative
    alphaBoolBlend {blendOption=Linear}
    boundBone {boneName=dyn_collar_l_01}
    componentLinearAccScale=<0.7 0.7 0.7>
    componentLinearVelScale=<0.7 0.7 0.7>
    gravityOverride {z=-980}
    physicsBodyDefinitions [{
      sphereCollisionRadius=0
      boundBone {boneName=dyn_collar_l_01}
      boxExtents=<20 30 30>
      constraintSetup {angularLimitsMax=<0 5 10>, angularLimitsMin=<0 -2 -10>, angularTarget=<1 0 0>}
      localJointOffset=<4 0 0>
     }]
   }
   n11 = Anim Dynamics - Bone: dyn_collar_l_02(ComponentPose=@n10, Alpha=1, ExternalForce="0,0,0") {
    angularDampingOverride=1
    angularSpringConstant=20000
    bAngularSpring=true
    bOverrideAngularDamping=true
    bUseGravityOverride=true
    gravityScale=0.35
    simulationSpace=RootRelative
    alphaBoolBlend {blendOption=Linear}
    boundBone {boneName=dyn_collar_l_02}
    componentLinearAccScale=<0.7 0.7 0.7>
    componentLinearVelScale=<0.7 0.7 0.7>
    gravityOverride {z=-980}
    physicsBodyDefinitions [{
      sphereCollisionRadius=0
      boundBone {boneName=dyn_collar_l_02}
      boxExtents=<20 30 30>
      constraintSetup {angularLimitsMax=<0 5 10>, angularLimitsMin=<0 -2 -10>, angularTarget=<1 0 0>}
      localJointOffset=<4 0 0>
     }]
   }
   n12 = Get Gravity Override()
   n13 = Anim Dynamics - Bone: dyn_skirt_bk_01!(ComponentPose=@n11, Alpha=1, GravityOverride:Vector=@n12, ExternalForce="0,0,0") -> (Pose) {
    angularBiasOverride=0.999
    angularDampingOverride=0.999
    angularSpringConstant=1500
    bOverrideAngularBias=true
    bOverrideAngularDamping=true
    bUseGravityOverride=true
    bUsePlanarLimit=false
    bUseSphericalLimits=true
    simulationSpace=RootRelative
    alphaBoolBlend {blendOption=Linear}
    alphaScaleBiasClamp {bClampResult=true, clampMax=0.4}
    boundBone {boneName=dyn_skirt_bk_01}
    componentAppliedLinearAccClamp=<3000 3000 3000>
    componentLinearAccScale=<0.4 0.5 0.5>
    componentLinearVelScale=<0.4 0.5 0.5>
    physicsBodyDefinitions [{
      collisionType=CustomSphere
      sphereCollisionRadius=2
      boundBone {boneName=dyn_skirt_bk_01}
      boxExtents=<8 17 17>
      constraintSetup {angularLimitsMax=<2.5 10 10>, angularLimitsMin=<-2.5 -10 -60>, angularTarget=<0 0 0>}
      localJointOffset=<3 0 0>
     }]
   }
   n14 = Anim Dynamics - Bone: dyn_skirt_bk_02!(ComponentPose=@n13, Alpha=1, GravityOverride:Vector=@n12, ExternalForce="0,0,0") -> (Pose) {
    angularBiasOverride=0.999
    angularDampingOverride=0.999
    angularSpringConstant=1500
    bOverrideAngularBias=true
    bOverrideAngularDamping=true
    bUseGravityOverride=true
    bUsePlanarLimit=false
    bUseSphericalLimits=true
    simulationSpace=RootRelative
    alphaBoolBlend {blendOption=Linear}
    alphaScaleBiasClamp {bClampResult=true, clampMax=0.4}
    boundBone {boneName=dyn_skirt_bk_02}
    componentAppliedLinearAccClamp=<3000 3000 3000>
    componentLinearAccScale=<0.4 0.5 0.5>
    componentLinearVelScale=<0.4 0.5 0.5>
    physicsBodyDefinitions [{
      collisionType=CustomSphere
      sphereCollisionRadius=4
      boundBone {boneName=dyn_skirt_bk_02}
      boxExtents=<8 17 17>
      constraintSetup {angularLimitsMax=<2.5 10 5>, angularLimitsMin=<-2.5 -10 -60>, angularTarget=<0 0 0>}
      localJointOffset=<3 0 0>
     }]
   }
   n15 = "Constraint - dyn_skirt_bk_01 (PP)"!(ComponentPose=@n14, Alpha=0.2, ConstraintWeights_0=0.5, ConstraintWeights_1:float=0.5) -> (Pose) {
    alphaBoolBlend {blendOption=Linear}
    boneToModify {boneName=dyn_skirt_bk_01}
    constraintSetup [{targetBone {boneName=thigh_l}}, {targetBone {boneName=thigh_r}}]
    constraintWeights [1,1]
   }
   n16 = "Constraint - dyn_skirt_bk_02 (PP)"!(ComponentPose=@n15, Alpha=0.2, ConstraintWeights_0=0.5, ConstraintWeights_1:float=0.5) -> (Pose) {
    alphaBoolBlend {blendOption=Linear}
    boneToModify {boneName=dyn_skirt_bk_02}
    constraintSetup [{targetBone {boneName=thigh_l}}, {targetBone {boneName=thigh_r}}]
    constraintWeights [1,1]
   }
   n17 = Control Rig!(Source=Component To Local(ComponentPose=@n16), Alpha, Skydive:double=Get Skydive() -> (Skydive:double)) -> (Pose) {
    alphaBoolBlend {blendOption=Linear}
    controlRigAssetReference {blueprintRigClass=&/F_MED_BlazerVeil_JacketOff_CtrlRig.F_MED_BlazerVeil_JacketOff_CtrlRig_C}
   }
   n18 = "Transform (Modify) Bone - Bone: dyn_tie_01"!(ComponentPose=Local To Component(LocalPose=@n17), Alpha=1, Rotation="0.005,3.000,-0.006") -> (Pose) {
    rotationMode=BMM_Additive
    rotationSpace=BCS_BoneSpace
    alphaBoolBlend {blendOption=Linear}
    boneToModify {boneName=dyn_tie_01}
    rotation=<rot 0.005 3 -0.006>
   }
   n19 = "Transform (Modify) Bone - Bone: dyn_tie_03"!(ComponentPose=@n18, Alpha=1, Rotation="-2,-7.743,0") -> (Pose) {
    rotationMode=BMM_Additive
    rotationSpace=BCS_BoneSpace
    alphaBoolBlend {blendOption=Linear}
    boneToModify {boneName=dyn_tie_03}
    rotation {pitch=-2, yaw=-7.743}
   }
   Output Pose(Result=Layered blend per bone(
     BasePose=Component To Local(ComponentPose=@n19)
     BlendPoses_0=Slot 'DefaultSlot'()
     BlendWeights_0=Get Custom Idle Is Playing() -> (Custom Idle Is Playing:float)
    ) {layerSetup [{branchFilters [{boneName=dyn_collar_l_01}, {boneName=dyn_collar_r_01}]}]}) {layerGroup=None}
  ]}
 EventGraph {nodes [
   Event BlueprintUpdateAnimation()
   > n1 = Sequence!(execute) -> (then_0, then_1, then_2:Exec)
   > n2 = Gravity Override State Function DEPRECATED(CharacterPartAnimInstance=Self-Reference(), GameTimeInSeconds=GetGameTimeInSeconds(), GravityOverrideParams=Get Gravity Override Params())
   > Set Gravity Override(Gravity Override=@n2)
   Set Skydive(execute=@n1.then_1, Skydive:double=SelectFloat(A=1, B=0, bPickA=Get bIsSkyDiving())) -> (Output_Get:double)
   Branch(execute=@n1.then_2, Condition=Get bIsInFrontEnd())
   > Set Custom Idle Is Playing(Custom Idle Is Playing:float=GetCurveValue(self=GetOwnerFrontEndAnimInstance(), CurveName=Frontend_Idle_Bypass)) -> (Output_Get:float)
  ]}
]
variables [
 Custom Idle Is Playing {type=number}
 Gravity Override {type=Vector}
 Gravity Override Params {type=GravityOverrideParamsStruct}
 Skydive {type=number}
]
